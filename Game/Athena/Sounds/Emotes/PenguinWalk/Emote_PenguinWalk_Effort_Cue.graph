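# Sound cue for the effort sounds of the Penguin Walk emote.
# A random node picks among six effort waves and two empty, equally weighted slots, so some triggers are silent.
# The pick passes through a short random delay, then a modulator that varies pitch and volume.
# It uses emote attenuation, the emote music sound class and concurrency set, and sends to a music audio-analysis submix.
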SoundCue /Game/Athena/Sounds/Emotes/PenguinWalk/Emote_PenguinWalk_Effort_Cue
root=Modulator_1
nodes [
 {id=Delay_0, children [Random_0], properties {delayMax=0.1}}
 {id=Modulator_1, children [Delay_0], properties {pitchMax=1.1, pitchMin=0.9, volumeMax=1.75, volumeMin=0.45}}
 {
  id=Random_0
  children [WavePlayer_0,WavePlayer_1,WavePlayer_2,WavePlayer_3,WavePlayer_4,WavePlayer_5,null,null]
  properties {bShouldExcludeFromBranchCulling=true, weights [1,1,1,1,1,1,1,1]}
 }
 {id=WavePlayer_0, properties {soundWaveAssetPtr=&/Emote_PenguinWalk_Effort_05}}
 {id=WavePlayer_1, properties {soundWaveAssetPtr=&/Emote_PenguinWalk_Effort_06}}
 {id=WavePlayer_2, properties {soundWaveAssetPtr=&/Emote_PenguinWalk_Effort_04}}
 {id=WavePlayer_3, properties {soundWaveAssetPtr=&/Emote_PenguinWalk_Effort_03}}
 {id=WavePlayer_4, properties {soundWaveAssetPtr=&/Emote_PenguinWalk_Effort_02}}
 {id=WavePlayer_5, properties {soundWaveAssetPtr=&/Emote_PenguinWalk_Effort_01}}
]
properties {
 attenuationSettings=&Fort_Emotes_Athena_Att
 maxDistance=1600
 soundClassObject=&Fort_Music_Emote
 volumeMultiplier=1
 concurrencyOverrides {maxCount=3, resolutionRule=StopOldest}
 concurrencySet [/Game/Sounds/Concurrency_Rules/Athena_EmoteMusic_Conc]
 soundSubmixSends [{minSendLevel=0, soundSubmix=&Fortnite_AudioAnalysis_Music}]
}
